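# Additive unlit material for soft spherical fog on Niagara mesh particles in the Fortnitemares pumpkin effects.
# A world-position texture sample perturbs a camera-to-normal dot product, which a SphereMask turns into soft-edged opacity.
# Particle alpha, pixel-depth near fade and DepthFade shape that opacity, while particle color feeds emissive adjustments.
# The same texture drives normal-offset WPO math and a perturbed, flattened normal behind a quality switch.
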Material /Game/Environments/Helios/FX/Fortnitemares/Pumpkins/Fog/Materials/M_UnlitTranslucency_SphereicalFog
graph {
 properties {
  bAutomaticallySetUsageInEditor=false
  bUsedWithNiagaraMeshParticles=true
  blendMode=BLEND_Additive
  shadingModel=MSM_Unlit
  translucencyLightingMode=TLM_VolumetricPerVertexDirectional
  usedShadingModels=Unlit
 }
 nodes [
  n1 = ParticleColor()
  n2 = TextureSample(UVs=Multiply(A=WorldPosition().XYZ, B=ScalarParameter() {defaultValue=0.0017, parameterName=Param_2})) {samplerType=SAMPLERTYPE_LinearColor, texture=/Game/Global/MaterialFunctions/VT_ClothSimWPO}
  n3 = ScalarParameter() {defaultValue=0.222, parameterName=Param_3}
  n4 = DotProduct(A=CameraVectorWS(), B=VertexNormalWS())
  n5 = SphereMask(
   A=Add(A=Multiply(A=@n2.RGB, B=@n3), B=Saturate(None=@n4))
   B=ScalarParameter() {defaultValue=0.9664, parameterName=Param_4}
   Radius=ScalarParameter() {defaultValue=0.444}
  ) {hardnessPercent=0}
  n6 = Divide(A=PixelDepth().R, B=ScalarParameter() {defaultValue=2500, parameterName=Param_1})
  DepthFade(Opacity=Multiply(A=Multiply(A=@n1.A, B=@n5), B=Saturate(None=@n6)), FadeDistance=ScalarParameter() {defaultValue=250, parameterName=Param_6})
  MaterialFunctionCall(Emissive=@n1.RGB) {materialFunction=&MF_FX_EmissiveAdjustments}
  Multiply() {constB=-1}
  Multiply(B=@n2.RGB)
  Multiply(A=Multiply(A=NamedRerouteUsage(), B=VertexNormalWS()), B=ScalarParameter() {defaultValue=50, parameterName=WPO Amount})
  NamedRerouteDeclaration(None=@n2.RGB) {name=WPO, nodeColor=<rgba 0 0.7188 1 1>}
  NamedRerouteDeclaration(None=@n2.RGB) {name=Normal, nodeColor=<rgba 0 1 0.5547 1>}
  OneMinus()
  OneMinus()
  OneMinus()
  n7 = NamedRerouteUsage()
  n8 = MaterialFunctionCall(
   Normal=Transform(None=MaterialFunctionCall!(Bump=@n7) {materialFunction=/Engine/Functions/Engine_MaterialFunctions03/Procedurals/PerturbNormalLQ}.WS Normal) {transformSourceType=TRANSFORMSOURCE_World, transformType=TRANSFORM_Tangent}
   Flatness=ScalarParameter() {parameterName=Param_7}
  ) {materialFunction=&FlattenNormal}
  n9 = Constant()
  QualitySwitch(Default=@n8.Result, Low=@n9, High=@n9, Medium=@n9, Epic=@n8.Result)
 ]
}
