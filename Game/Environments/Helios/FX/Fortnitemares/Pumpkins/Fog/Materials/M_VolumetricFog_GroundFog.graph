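# Additive volume-domain material for Niagara particles, producing ground fog for the Helios Fortnitemares pumpkin effects.
# A world-space sample of the VT_ClothSimWPO texture offsets world Z, which is subtracted from particle Z to form a saturated height mask with an uneven top.
# That mask is scaled by particle alpha and a parameter, and also combines with the texture to drive a separate interpolation.
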Material /Game/Environments/Helios/FX/Fortnitemares/Pumpkins/Fog/Materials/M_VolumetricFog_GroundFog
graph {
 properties {
  bAutomaticallySetUsageInEditor=false
  bUsedWithNiagaraSprites=true
  blendMode=BLEND_Additive
  materialDomain=MD_Volume
  usedShadingModels=Default Lit
 }
 nodes [
  DistanceToNearestSurface!(World Position)
  n1 = TextureSample(UVs=Multiply(A=WorldPosition().XYZ, B=ScalarParameter() {defaultValue=0.003, parameterName=Param_3})) {samplerType=SAMPLERTYPE_LinearColor, texture=/Game/Global/MaterialFunctions/VT_ClothSimWPO}
  n2 = ComponentMask(None=ParticlePositionWS()) {b=true}
  n3 = Add(A=Multiply(A=@n1.RGB, B=ScalarParameter() {defaultValue=150, parameterName=Param_1}), B=WorldPosition().Z)
  n4 = Saturate(None=Subtract(A=Subtract(A=@n2, B=@n3), B=ScalarParameter() {defaultValue=4, parameterName=Param_4}))
  LinearInterpolate(B=ScalarParameter() {defaultValue=0.01}, Alpha=Multiply(A=@n1.RGB, B=@n4)) {constB=0.1}
  Multiply(A=Multiply(A=ParticleColor().A, B=@n4), B=ScalarParameter() {defaultValue=1, parameterName=Param_2})
  OneMinus()
  TextureCoordinate()
  WorldPosition()
 ]
}
