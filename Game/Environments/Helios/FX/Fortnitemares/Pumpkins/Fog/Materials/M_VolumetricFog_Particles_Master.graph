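# Master volume-domain material for Niagara particles that draw volumetric fog, from the Fortnitemares pumpkin effects.
# A world-space noise texture, distorted by time-animated curl noise, breaks up the fog.
# Its inverted green channel scales the particle color, while its RGB blends between min and max fog noise values.
# That clamped noise, a soft sphere mask at particle position and radius, and particle alpha multiply into density.
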Material /Game/Environments/Helios/FX/Fortnitemares/Pumpkins/Fog/Materials/M_VolumetricFog_Particles_Master
graph {
 properties {
  bAutomaticallySetUsageInEditor=false
  bUsedWithNiagaraSprites=true
  blendMode=BLEND_Additive
  materialDomain=MD_Volume
  previewMesh=/Engine/Transient.MaterialEditorMeshComponent_1
  usedShadingModels=Default Lit
 }
 nodes [
  n1 = ParticleColor()
  n2 = WorldPosition()
  n3 = ScalarParameter() {defaultValue=0.0025, parameterName=Fog Noise Frequency}
  n4 = ScalarParameter() {defaultValue=1, parameterName=Curl Amplitude}
  n5 = Multiply(A=Time())
  n6 = WorldPosition()
  n7 = TextureSample(UVs=Multiply(A=Add(A=@n5, B=@n6.XYZ), B=ScalarParameter() {defaultValue=1, parameterName=Curl Frequency})) {samplerType=SAMPLERTYPE_LinearColor, texture=/Game/Effects/FogVolume/VT_CurlNoise_LOW}
  n8 = TextureSample(UVs=Add(A=Multiply(A=@n2.XYZ, B=@n3), B=Multiply(A=@n4, B=@n7.RGB))) {samplerType=SAMPLERTYPE_LinearColor, texture=/Game/Global/MaterialFunctions/VT_ClothSimWPO}
  LinearInterpolate(B=@n1.RGB, Alpha=OneMinus(None=@n8.G))
  n9 = WorldPosition()
  n10 = ParticlePositionWS()
  n11 = ParticleRadius()
  n12 = LinearInterpolate(A=ScalarParameter() {parameterName=Fog Noise Min}, B=ScalarParameter() {defaultValue=2, parameterName=Fog Noise Max}, Alpha=@n8.RGB)
  Multiply(A=@n1.A, B=Multiply(A=SphereMask(A=@n9.XYZ, B=@n10, Radius=@n11) {hardnessPercent=0}, B=Saturate(None=@n12)))
 ]
}
