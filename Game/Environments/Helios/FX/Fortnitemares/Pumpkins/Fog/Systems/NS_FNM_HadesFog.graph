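# Niagara fog effect for the Fortnitemares pumpkin environment, drawing faint, slowly spawned, red-tinted particles through two sprite renderers and one mesh renderer.
# A single emitter scatters particles over a wide, flat, owner-scaled box and pushes them along X with acceleration and drag.
# LOD distance remaps shrink sprites to zero and meshes to a fraction at close range.
# A low-scaled alpha curve fades each particle in and out over its lifetime.
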NiagaraSystem /Game/Environments/Helios/FX/Fortnitemares/Pumpkins/Fog/Systems/NS_FNM_HadesFog
dependencies {
 usedDataInterfaces [&NiagaraDataInterfaceCurve, &NiagaraDataInterfaceColorCurve]
 usedDynamicInputs []
 usedModules [&SystemState]
}
emitters {Empty {
  emitterUpdateScript [
   &EmitterState()
   &SpawnRate(SpawnRate=4, Use Spawn Probability=false)
  ]
  particleSpawnScript [
   &InitializeParticle(
    Color=<rgba 0.5 0.1979 0.1979 1>
    Lifetime=5
    Mesh Scale Mode=::NewEnumerator1
    Mesh Uniform Scale Max=12
    Mesh Uniform Scale Min=2
    Sprite Size Mode=::NewEnumerator1
    Uniform Sprite Size Max=750
    Uniform Sprite Size Min=500
    UsePositionOffset=false
    Write Material Random=true
   )
   &ShapeLocation(Box Size=<3500 3500 0>, MaskSpawnGroup=false, Override Seed=false, Shape Primitive=::NewEnumerator2, Surface Only=false, UseOwnerScale=true)
  ]
  particleUpdateScript [
   &ParticleState()
   &ScaleMeshSize(Scale Factor=&Lerp_Vector(
     A=<0.125 0.125 0.125>
     B=<1 1 1>
     Alpha=&FN_LOD_Distance_Remap(Start Distance=0, End Distance=3500, Minimum Output=0, Maximum Output=1, Clamp Output=true)
    ))
   &ScaleSpriteSize(
    Uniform Curve Index=&FN_LOD_Distance_Remap(Start Distance=0, End Distance=1500, Minimum Output=0, Maximum Output=1, Clamp Output=true)
    Uniform Curve Sprite Scale=DataInterface() {propertyValues {Curve {keys [{interpMode=RCIM_Cubic, tangentMode=RCTM_Break, time=0, value=0}, {tangentMode=RCTM_User, time=1, value=1}]}}}
   )
   &ScaleColor(
    Scale Alpha=&FloatFromCurve(
     FloatCurve=DataInterface() {propertyValues {Curve {keys [
         {interpMode=RCIM_Cubic, tangentMode=RCTM_User, time=0, value=0}
         {interpMode=RCIM_Cubic, tangentMode=RCTM_User, time=0.5, value=1}
         {tangentMode=RCTM_User, time=1, value=0}
        ]}}}
     Scale Curve=0.075
    )
    ScaleA=true
    ScaleRGB=true
   )
   &AccelerationForce(Acceleration=<250 0 0>)
   &Drag(Use Linear Drag=true, Use Rotational Drag=true)
   &SolveForcesAndVelocity(Clamp Velocity=false, Limit Acceleration=false)
  ]
  rendererClasses [&NiagaraSpriteRendererProperties, &NiagaraMeshRendererProperties]
  renderers [
   {rendererClass=&NiagaraSpriteRendererProperties}
   {rendererClass=&NiagaraMeshRendererProperties, rendererIndex=1}
   {rendererClass=&NiagaraSpriteRendererProperties, rendererIndex=2}
  ]
 }}
summary {
 systemName=NS_FNM_HadesFog
 emitters [{emitterName=Empty, rendererClasses [&NiagaraSpriteRendererProperties, &NiagaraMeshRendererProperties]}]
 userVariables []
}
